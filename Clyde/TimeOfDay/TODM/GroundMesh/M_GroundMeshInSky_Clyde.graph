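# Unlit sky material for a Clyde time-of-day ground mesh, faking distant terrain with a tiled, desaturated ground texture tinted by Color_Time_of_Day.
# Normal-map shading against the atmosphere light adds relief, unless the UseDiffuseOnly static switch bypasses it.
# A UV-centered sphere mask applies a darkened, desaturated shadow spot.
# Near-horizontal view angles fade into time-of-day-blended sky capture textures, sampled by camera vector, hiding the horizon seam.
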Material /Clyde/TimeOfDay/TODM/GroundMesh/M_GroundMeshInSky_Clyde
graph {
 properties {bIsSky=true, bTangentSpaceNormal=false, pixelDepthOffsetMode=PDOM_Legacy, shadingModel=MSM_Unlit, usedShadingModels=Unlit}
 nodes [
  n1 = Multiply(A=TextureCoordinate(), B=ScalarParameter() {defaultValue=60, parameterName=tile})
  n2 = Desaturation(
   None=TextureSample(UVs=@n1) {texture=/Game/TimeOfDay/TODM/BR/s11/GroundMesh/T_ReflectionGround}.RGB
   Fraction=ScalarParameter() {defaultValue=0.8, parameterName=desat}
  )
  n3 = MaterialFunctionCall(
   Color_Morning=VectorParameter() {parameterName=ColorMorning, defaultValue=<rgba 0.5312 0.3966 0.4178 1>}.RGB
   Color_Day=VectorParameter() {parameterName=ColorDay, defaultValue=<rgba 0.6198 0.6001 0.3834 1>}.RGB
   Color_Evening=VectorParameter() {parameterName=ColorEvening, defaultValue=<rgba 0.5 0.1961 0.075 1>}.RGB
   Color_Night=VectorParameter() {parameterName=ColorNight, defaultValue=<rgba 0.2448 0.2608 0.3594 1>}.RGB
  ) {materialFunction=&Color_Time_of_Day}
  n4 = Multiply(A=Desaturation(None=@n3.Result)) {constB=0.1}
  n5 = DotProduct(
   A=TextureSample(UVs=@n1) {samplerType=SAMPLERTYPE_Normal, texture=/Game/TimeOfDay/TODM/BR/s11/GroundMesh/T_ReflectionGround_N}.RGB
   B=SkyAtmosphereLightDirection()
  )
  n6 = Multiply(A=Clamp(None=ConstantBiasScale(None=@n5)), B=@n3.Result)
  n7 = StaticSwitchParameter(True=Add(A=Multiply(A=@n2, B=@n3.Result), B=@n4), False=Multiply(A=@n2, B=Add(A=@n6, B=@n4))) {parameterName=UseDiffuseOnly}
  n8 = Desaturation(None=@n7, Fraction=ScalarParameter() {defaultValue=1, parameterName=ShadowDesat})
  n9 = ScalarParameter() {defaultValue=0.5, parameterName=ShadowBrightness}
  n10 = Constant() {r=0.5}
  n11 = TextureCoordinate()
  n12 = ScalarParameter() {defaultValue=0.015, parameterName=rad}
  n13 = ScalarParameter() {defaultValue=0.1, parameterName=h}
  n14 = Multiply(A=CameraVectorWS(), B=Constant3Vector() {constant=<rgba -1 -1 1 0>})
  n15 = Abs(None=ComponentMask(None=CameraVectorWS()) {b=true})
  LinearInterpolate(
   A=LinearInterpolate(A=@n7, B=Multiply(A=@n8, B=@n9), Alpha=SphereMask(A=@n10, B=@n11, Radius=@n12, Hardness=@n13))
   B=MaterialFunctionCall(
    Color_Morning=TextureSample(UVs=@n14) {samplerType=SAMPLERTYPE_LinearColor, texture=&/T_SkyLightCapture_Sunrise_Clyde}.RGB
    Color_Day=TextureSample(UVs=@n14) {samplerType=SAMPLERTYPE_LinearColor, texture=&/T_SkyLightCapture_Day_Clyde}.RGB
    Color_Evening=TextureSample(UVs=@n14) {samplerType=SAMPLERTYPE_LinearColor, texture=&/T_SkyLightCapture_Sunset_Clyde}.RGB
    Color_Night=TextureSample(UVs=@n14) {samplerType=SAMPLERTYPE_LinearColor, texture=&/T_SkyLightCapture_Night_Clyde}.RGB
   ) {materialFunction=&Color_Time_of_Day}.Result
   Alpha=Power(Base=OneMinus(None=@n15)) {constExponent=8}
  )
  SkyLightEnvMapSample(Direction=@n14, Roughness=Constant())
 ]
}
